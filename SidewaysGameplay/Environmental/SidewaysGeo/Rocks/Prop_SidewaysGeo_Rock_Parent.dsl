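; Parent Blueprint for Sideways rock pillar props that fade in and out with sound when dynamically spawned.
; BeginPlay triggers the fade-in, swapping in a dynamic fade material driven by game time and a reverse flag.
; After a short delay the server sets FadedIn to prevent repeats, while clients later restore the static material.
; DestroyDynamicPillar reapplies the fade material in reverse, waits the fade-out delay, then calls SilentDie.
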(event EventBeginPlay
  ;[1 1 1] Begin Play - Check if it is a dynamic spawn. If so begin the fade.
  (if (and (Variables|Default|GetDynamicSpawn) (not (Variables|Default|GetFadedIn)))
    (Development|SetHiddeninGame (Variables|Debug|GetStaticMeshComponent) true false)
    (CallFunction|SidewaysFadeBegin (Transformation|GetActorLocation (Variables|Default|GetCoreReference)))))

(event Custom|SidewaysFadeBegin
  (params (CoreLocation :type Vector))
  (bind _staticmeshcomponent (Variables|Debug|GetStaticMeshComponent))
  ;[1 1 1] Sideways Fade Begin - Fade in the mesh of the pillars
  (Development|SetHiddeninGame _staticmeshcomponent false false)
  (Variables|Default|SetStaticMaterial (Rendering|Material|GetMaterial _staticmeshcomponent 0))
  (bind _returnvalue (Rendering|Material|CreateDynamicMaterialInstance _staticmeshcomponent 0 (Variables|Default|GetFadeMaterial) "None" :!class /Script/Engine.PrimitiveComponent))
  (bind _output_get (Variables|Default|SetMID _returnvalue))
  (Rendering|Material|SetScalarParameterValue _output_get "UseManualFade" 0.0 :!class /Script/Engine.MaterialInstanceDynamic)
  (Rendering|Material|SetScalarParameterValue _output_get "ReverseFade" 0.0 :!class /Script/Engine.MaterialInstanceDynamic)
  (Rendering|Material|SetScalarParameterValue _output_get "BP_FadeTimeBegin" (Utilities|Time|GetGameTimeinSeconds) :!class /Script/Engine.MaterialInstanceDynamic)
  (Audio|Components|Audio|Play (Variables|Default|GetPillar_Spawn_Sound) 0.0)
  ;;[1 1 1] After A delay set a value to ensure we do not hit a race condition but also so that we never call this from begin play again
  (Utilities|FlowControl|Delay 0.1)
  (if (Networking|IsServer)
    (Variables|Default|SetFadedIn true)
    (else
      (Utilities|FlowControl|Delay (Variables|Default|GetDelayForFadingOut))
      (Rendering|Material|SetMaterial (Variables|Debug|GetStaticMeshComponent) 0 (Variables|Default|GetStaticMaterial)))))

(event Custom|SidewaysFadeOut
  (params (CoreLocation :type Vector))
  (bind _mid (Variables|Default|GetMID))
  ;[1 1 1] Sideways Fade Out - Fade out the mesh of the pillars - Note the fade out delay needs to match this fade time (roughly)
  (Rendering|Material|SetMaterial (Variables|Debug|GetStaticMeshComponent) 0 _mid)
  (Rendering|Material|SetScalarParameterValue _mid "UseManualFade" 0.0 :!class /Script/Engine.MaterialInstanceDynamic)
  (Rendering|Material|SetScalarParameterValue _mid "ReverseFade" 1.0 :!class /Script/Engine.MaterialInstanceDynamic)
  (Rendering|Material|SetScalarParameterValue _mid "BP_FadeTimeBegin" (Utilities|Time|GetGameTimeinSeconds) :!class /Script/Engine.MaterialInstanceDynamic)
  (Audio|Components|Audio|Play (Variables|Default|GetPillar_Remove_Sound) 0.0))

(event Custom|DestroyDynamicPillar
  ;[1 1 1] Destory Dynamic Pillar - Call the fade out then after a delay call Silent Die
  (if (Variables|Default|GetDynamicSpawn)
    (CallFunction|SidewaysFadeOut (Transformation|GetActorLocation (Variables|Default|GetCoreReference)))
    (Utilities|FlowControl|Delay (Variables|Default|GetDelayForFadingOut))
    (Gameplay|SilentDie false)))

(fn ConstructionScript)
